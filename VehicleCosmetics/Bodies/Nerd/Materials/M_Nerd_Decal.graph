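# M_Nerd_Decal is a masked, two-sided Default Lit material, named as the decal material for the Nerd vehicle body cosmetic.
# Its graph holds a single texture sample of the Nerd decal normal map, with no output connections shown.
# Usage is enabled for skeletal meshes and mesh deformers, and automatic usage setting in the editor is turned off.
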Material /VehicleCosmetics/Bodies/Nerd/Materials/M_Nerd_Decal
graph {
 properties {
  bAutomaticallySetUsageInEditor=false
  bUsedWithMeshDeformer=true
  bUsedWithSkeletalMesh=true
  blendMode=BLEND_Masked
  twoSided=true
  usedShadingModels=Default Lit
 }
 nodes [TextureSample() {samplerType=SAMPLERTYPE_Normal, texture=/VehicleCosmetics/Bodies/Nerd/Textures/T_nerd_decal_normal}]
}
